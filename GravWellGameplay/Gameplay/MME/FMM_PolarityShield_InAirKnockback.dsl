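; Guided-motion movement mode logic that sets the constant-force direction for Polarity Shield in-air knockback, pushing the owner opposite their aim.
; GetDirection_ConstantForce takes the net-synced base aim rotation, adds a tunable pitch offset, and negates the forward vector.
; It scales that vector by a tuned magnitude and adds extra Z.
; Both terms are safe-divided by the configured ConstantForce Strength, so the returned direction also carries magnitude.
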(fn GetDirection_ConstantForce
  (params (GuidedMotionRuntimeData :type Fort%20Movement%20Mode%20Guided%20Motion%20Runtime%20Data%20Object%20Reference))
  (bind (_returnvalue_x _returnvalue_y _returnvalue_z) (Math|Vector|BreakVector (Math|Vector|Normalize (* (Math|Vector|GetForwardVector (Math|Rotator|MakeRotator (.roll (CharacterMovement|GetNetSyncedBaseAimRotation (Variables|FortMovementMode_BaseExtLogic|GetOwnerMovementComponent))) (+ (.pitch (CharacterMovement|GetNetSyncedBaseAimRotation (Variables|FortMovementMode_BaseExtLogic|GetOwnerMovementComponent))) (ScalableFloat|GetValueatLevel (Variables|Default|GetHF_OffsetPitch))) (.yaw (CharacterMovement|GetNetSyncedBaseAimRotation (Variables|FortMovementMode_BaseExtLogic|GetOwnerMovementComponent))))) -1))))
  (return (* _returnvalue_x (Math|Float|SafeDivide (ScalableFloat|GetValueatLevel (Variables|Default|GetHF_Mag)) (ScalableFloat|GetValueasDouble (Variables|Config-ConstantForce|GetStrength)))) (* _returnvalue_y (Math|Float|SafeDivide (ScalableFloat|GetValueatLevel (Variables|Default|GetHF_Mag)) (ScalableFloat|GetValueasDouble (Variables|Config-ConstantForce|GetStrength)))) (+ (* _returnvalue_z (Math|Float|SafeDivide (ScalableFloat|GetValueatLevel (Variables|Default|GetHF_Mag)) (ScalableFloat|GetValueasDouble (Variables|Config-ConstantForce|GetStrength)))) (Math|Float|SafeDivide (ScalableFloat|GetValueatLevel (Variables|Default|GetHF_ZAdd)) (ScalableFloat|GetValueasDouble (Variables|Config-ConstantForce|GetStrength))))))
