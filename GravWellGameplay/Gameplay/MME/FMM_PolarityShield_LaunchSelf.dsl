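; FMM_PolarityShield_LaunchSelf is a Fortnite guided-motion movement mode that, by its name, launches the owner during the Polarity Shield ability.
; Its GetDirection_ConstantForce function takes X and Y from the owner's normalized movement acceleration, scaled by an XY launch force, and Z from a separate launch force.
; All axes divide by the constant-force Strength, so the unnormalized direction vector appears to cancel that multiplier and carry absolute horizontal and vertical forces.
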(fn GetDirection_ConstantForce
  (params (GuidedMotionRuntimeData :type Fort%20Movement%20Mode%20Guided%20Motion%20Runtime%20Data%20Object%20Reference))
  (bind (_returnvalue_x _returnvalue_y) (Math|Vector|BreakVector (Math|Vector|Normalize (Pawn|Components|CharacterMovement|GetCurrentAcceleration (Variables|FortMovementMode_BaseExtLogic|GetOwnerMovementComponent)))))
  (return (* _returnvalue_x (Math|Float|SafeDivide (ScalableFloat|GetValueatLevel (Variables|Default|GetHF_XYLaunchForce)) (ScalableFloat|GetValueasDouble (Variables|Config-ConstantForce|GetStrength)))) (* _returnvalue_y (Math|Float|SafeDivide (ScalableFloat|GetValueatLevel (Variables|Default|GetHF_XYLaunchForce)) (ScalableFloat|GetValueasDouble (Variables|Config-ConstantForce|GetStrength)))) (Math|Float|SafeDivide (ScalableFloat|GetValueatLevel (Variables|Default|GetHF_ZLaunchForce)) (ScalableFloat|GetValueasDouble (Variables|Config-ConstantForce|GetStrength)))))
